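# Niagara idle effect for the FacetWhisk cosmetic companion, shedding slowly tumbling mesh shards near its tail and bandana.
# Two emitters spawn at skeletal sockets through a shared socket reader, inherit some velocity, then drift with drag while growing and shrinking over life.
# User.VariantID selects bandana shard color and spawn rate, with variant 0 emitting none; User.DetailVariantID selects tail shard color.
# Tail shards also use distance-based kill.
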NiagaraSystem /CosmeticCompanions/FX/Builtins/FacetWhisk/NS_Companion_FacetWhisk_Idle
dependencies {
 usedDataInterfaces [
  &NiagaraDataInterfaceMeshRendererInfo
  &NiagaraDataInterfaceCurve
  &NiagaraDataInterfaceSocketReader
 ]
 usedDynamicInputs [&SetBoolByFloatComparison]
 usedModules [
  &self:SystemSpawnScript.SystemScriptSource.SystemScriptGraph.NiagaraNodeAssignment_1.SetVariables_#1
  &self:SystemSpawnScript.SystemScriptSource.SystemScriptGraph.NiagaraNodeAssignment_0.SetVariables_#2
  &SystemState
 ]
}
emitters {
 Shards_Bandana {
  emitterUpdateScript [
   &EmitterState()
   &SpawnRate(
    SpawnRate=&MergeHelper_FloatSwitch(0=0, 1=1.3, 2=1.3, 3=?, Index=&MakeIntFromFloat(Float=$User.VariantID), Num=::NewEnumerator1)
    Use Spawn Probability=false
   )
  ]
  particleSpawnScript [
   &InitializeParticle(
    Color=&MergeHelper_ColorSwitch(
     0=<rgba 0 0 0 0>
     1=<rgba 1 1 1 1>
     2=<rgba 0.1313 0.125 0.5 1>
     3=?
     Index=&MakeIntFromFloat(Float=$User.VariantID)
     Num=::NewEnumerator1
    )
    Lifetime Max=2.5
    Lifetime Min=1.5
    Lifetime Mode=::NewEnumerator1
    Mesh Scale Mode=::NewEnumerator1
    Mesh Uniform Scale Max=1.1
    Mesh Uniform Scale Min=0.8
    Sprite Size Mode=::NewEnumerator1
    UsePositionOffset=false
    Write Material Random=true
   )
   &InitialMeshOrientation(
    Mesh Orientation Mode=::NewEnumerator2
    Rotation=&RandomRangeVector(Minimum=<0.1 0 0>, Maximum=<0.15 1 0>)
    Use Rotation Vector=true
   )
   &SocketLocation(
    "Inherit Velocity (Scale)"=0.2
    Orientation Sampling=::NewEnumerator1
    SamplePosOffset=true
    Sampled Orientation Coordinate Space:ENiagaraCoordinateSpace=::Local
    Sampled Position Offset=<0 -7 6>
    Socket Index=5
    Socket Reader=$System.Socket Reader
    Socket Sampling Mode=::NewEnumerator3
    VelocitySpeedLimit=false
   )
   &AddVelocity(
    Cone Angle=180
    Cone Axis=<0 -1 0>
    Module.UseConeFalloff=false
    Module.UseVelDistribution=false
    Override Seed=false
    Velocity Mode=::NewEnumerator2
    Velocity Speed=&RandomRangeFloat(Minimum=9, Maximum=14)
   )
   &ShapeLocation(MaskSpawnGroup=false, Override Seed=false, Sphere Radius=2, UseOwnerScale=true)
   /SetVariables(Particles.MeshIndex=&RandomRangeInt(Minimum=0, Maximum=2)) {isSetParameters=true}
  ]
  particleUpdateScript [
   &ParticleState()
   &UpdateMeshOrientation(Rotation Rate=&RandomRangeFloat(Minimum=0.4, Maximum=0.8), Rotation Vector=<0.1 0.1 0.1>)
   &ScaleMeshSize(Scale Factor=&VectorFromFloat(Value=&FloatFromCurve(FloatCurve=DataInterface() {propertyValues {Curve {
         defaultValue=0
         keys [
          {interpMode=RCIM_Cubic, tangentMode=RCTM_User, time=0, value=0}
          {interpMode=RCIM_Cubic, tangentMode=RCTM_User, time=0.5, value=1}
          {tangentMode=RCTM_User, time=1, value=0}
         ]
        }}})))
   &Drag(Drag=&RandomRangeFloat(Minimum=0.3, Maximum=0.6), Use Linear Drag=true, Use Rotational Drag=true)
   &SolveForcesAndVelocity(Clamp Velocity=false, Limit Acceleration=false)
  ]
  rendererClasses [&NiagaraMeshRendererProperties]
  renderers [
   {rendererClass=&NiagaraMeshRendererProperties, rendererIndex=0}
   {rendererClass=&NiagaraMeshRendererProperties, rendererIndex=1}
  ]
 }
 Shards_Tail {
  emitterUpdateScript [
   &EmitterState()
   &SpawnRate(SpawnRate=2, Use Spawn Probability=false)
  ]
  particleSpawnScript [
   &InitializeParticle(
    Color=&MergeHelper_ColorSwitch(
     0=<rgba 1 0 0.8859 1>
     1=<rgba 0.3068 0.8881 1 1>
     2=<rgba 0.8 0.3613 0.048 1>
     3=<rgba 0.6957 0.0569 0.0548 1>
     Index=&MakeIntFromFloat(Float=$User.DetailVariantID)
     Num=::NewEnumerator2
    )
    Lifetime Max=2.5
    Lifetime Min=1.5
    Lifetime Mode=::NewEnumerator1
    Mesh Renderer Array Visibility Mode=::NewEnumerator3
    Mesh Renderer Info=DataInterface() {propertyValues {MeshRenderer=&self:Shards_Tail_0.NiagaraMeshRendererProperties_1}}
    Mesh Scale Mode=::NewEnumerator1
    Mesh Uniform Scale Max=1
    Mesh Uniform Scale Min=0.5
    Sprite Size Mode=::NewEnumerator1
    UsePositionOffset=false
    Write Material Random=true
   )
   &InitialMeshOrientation(
    Mesh Orientation Mode=::NewEnumerator2
    Rotation=&RandomRangeVector(Minimum=<0 0 0>, Maximum=<0 0 1>)
    Use Rotation Vector=true
   )
   &SocketLocation(
    "Inherit Velocity (Scale)"=0.1
    Orientation Sampling=::NewEnumerator1
    SamplePosOffset=false
    Sampled Orientation Coordinate Space:ENiagaraCoordinateSpace=::Local
    Socket Index=&RandomRangeInt(Minimum=4, Maximum=4)
    Socket Reader=$System.Socket Reader
    Socket Sampling Mode=::NewEnumerator3
    Velocity Sampling:ENiagara_AttributeSamplingApplyOutput=::NewEnumerator0
    VelocitySpeedLimit=false
   )
   &AddVelocity(
    Module.Use Falloff Distance=false
    Override Seed=false
    Velocity Mode=::NewEnumerator1
    Velocity Origin=$Output.SocketLocation.SampledPosition
    Velocity Speed=&RandomRangeFloat(Minimum=5, Maximum=15)
   )
  ]
  particleUpdateScript [
   &ParticleState()
   &UpdateMeshOrientation(Rotation Rate=&RandomRangeFloat(Minimum=0.4, Maximum=1.2), Rotation Vector=<0.2 0.3 0>)
   &ScaleMeshSize(Scale Factor=&VectorFromFloat(Value=&FloatFromCurve(FloatCurve=DataInterface() {propertyValues {Curve {
         defaultValue=0
         keys [
          {interpMode=RCIM_Cubic, tangentMode=RCTM_User, time=0, value=0}
          {interpMode=RCIM_Cubic, tangentMode=RCTM_User, time=0.5, value=1}
          {tangentMode=RCTM_User, time=1, value=0}
         ]
        }}})))
   &ScaleColor(
    Scale RGB=&MakeVectorFromLinearColorRGB(Color=&FNTimeOfDayColor(Night Color=<rgba 1 1 1 1>, Morning Color=<rgba 1 1 1 1>, Day Color=<rgba 1 1 1 1>, Evening Color=<rgba 1 1 1 1>))
    ScaleA=true
    ScaleRGB=true
   )
   &Drag(Drag=&RandomRangeFloat(Minimum=0.15, Maximum=0.3), Use Linear Drag=true, Use Rotational Drag=true)
   &SolveForcesAndVelocity(Clamp Velocity=false, Limit Acceleration=false)
   /Game/Effects/Niagara/Emitters/Water/FN_DistanceKillParticles(AgeReduiceAmount=0.2, DistanceFalloffExp=0.3)
  ]
  rendererClasses [&NiagaraMeshRendererProperties]
  renderers [{rendererClass=&NiagaraMeshRendererProperties, rendererIndex=0}]
 }
}
summary {
 systemName=NS_Companion_FacetWhisk_Idle
 emitters [
  {emitterName=Shards_Tail, rendererClasses [&NiagaraMeshRendererProperties]}
  {emitterName=Shards_Bandana, rendererClasses [&NiagaraMeshRendererProperties]}
 ]
 userVariables [
  User.DetailVariantID {type=NiagaraFloat, defaultValue=0, description=""}
  User.VariantID {type=NiagaraFloat, defaultValue=0, description=""}
 ]
}
